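; Gameplay ability for a slide-time quest, named for 10 seconds, that credits the player after enough elapsed time.
; On activation it waits for quest progression, then loops a timer that adds its interval to ElapsedTimeInAir, without checking slide state itself.
; At RequiredTime it sends the T_Quest stat event to the quest manager and ends.
; Ending resets the total and clears the timer; air-time names and comments remain.
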(event Ability|EventActivateAbilityFromEvent
  (params (EventData :type Gameplay%20Event%20Data%20Structure%20%28by%20ref%29))
  (|CallParentFunction EventData)
  (Class|FortAsyncActionWaitQuestProgressionEnabled|WaitQuestProgressionEnabled
    (:OnQuestProgressionEnabled
      (Development|PrintString "Sliding satarted")
      ;[1 1 1] Start Checking Vehicle If Vehicle is in Air
      (bind _returnvalue (Utilities|Time|SetTimerbyEvent (delegate Custom|Check%20Slide%20Time) (Variables|Default|GetExecutionTimeInterval) true))
      (Variables|Default|SetTimer%20Handle _returnvalue))))

(event Custom|Check%20Slide%20Time
  ;[1 1 1] Add the time spent in air to the total elapsed time in air (as calculated by the time interval)
  (bind _output_get (Variables|Default|SetElapsedTimeInAir (+ (Variables|Default|GetExecutionTimeInterval) (Variables|Default|GetElapsedTimeInAir))))
  (if (>= _output_get (Variables|Default|GetRequiredTime))
    (Quest|SendComplexCustomStatEvent (Variables|GAVars|Data|GetOwningPlayerQuestManager) 0 (Variables|Default|GetT_Quest) "(GameplayTags=)" 1)
    (Utilities|Time|ClearandInvalidateTimerbyHandle (Variables|Default|GetTimer%20Handle))
    (Ability|EndAbility)))

(event Ability|EventOnEndAbility
  (params (bWasCancelled :type Boolean))
  (bind _timer_handle (Variables|Default|GetTimer%20Handle))
  ;[1 1 1] On End Ability
  (Variables|Default|SetElapsedTimeInAir 0.0)
  (if (Utilities|Time|IsValidTimerHandle _timer_handle)
    (Utilities|Time|ClearandInvalidateTimerbyHandle _timer_handle)
    (else
      _))
  (Ability|EndAbility))
